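# Material function that adds multi-colored, animated emissive lights to winter prop materials, likely Christmas trees, behind a static switch.
# The second texture coordinate index's V value selects a tinted gradient or one of two solid glow colors.
# The gradient's UVs (index 1 or 2) jump once per second via ceiled Time.
# A texture's red channel masks the glow, with optional camera distance fade, output alone or blended into input material attributes.
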MaterialFunction /FigmentEnvironment/LM_8_1_1/Environments/Props/Winter/Materials/EmissiveTextureWinterTree
graph {nodes [
  n1 = TextureSampleParameter2D() {
   desc="spec, metallic,roughness\n"
   group=Emissive_Winter
   parameterName=EmissiveTexture
   samplerType=SAMPLERTYPE_Masks
   texture=/FigmentEnvironment/LM_8_1_1/Packages/Fortnite_UniqueMaterials_1/Textures/T_Fortnite_Default_S
  }
  n2 = ComponentMask(None=TextureCoordinate() {coordinateIndex=2}) {g=true}
  n3 = TextureCoordinate() {coordinateIndex=1}
  n4 = TextureCoordinate() {coordinateIndex=2}
  n5 = Time()
  n6 = Add(
   A=Multiply(
    A=StaticSwitchParameter(True=@n3, False=@n4) {parameterName=UseEmissiveAnimationUV1}
    B=ScalarParameter() {defaultValue=-2.8638, group=Emissive_Winter, parameterName=TilingEmissiveColorVar}
   )
   B=Multiply(A=ScalarParameter() {defaultValue=1.3514, group=Emissive_Winter, parameterName=SpeedEmissiveColorVar_1}, B=Ceil(None=@n5))
  )
  n7 = Multiply(
   A=Multiply(
    A=VectorParameter() {group=Emissive_Winter, parameterName=ColorEmissive_Winter, defaultValue=<rgba 1 1 1 1>}.RGB
    B=TextureSample(UVs=@n6) {samplerType=SAMPLERTYPE_Masks, texture=/FigmentEnvironment/LM_8_1_1/Environments/Props/Winter/Textures/T_XMas_LightGradient}.RGB
   )
   B=ScalarParameter() {defaultValue=50, parameterName=Glow}
  )
  n8 = If(
   A=@n2
   B=Constant() {r=1}
   "A > B"=Multiply(
    A=VectorParameter() {group=Emissive_Winter, parameterName=ColorEmissive_Winter_1, defaultValue=<rgba 1 0.7494 0.2031 1>}.RGB
    B=ScalarParameter() {defaultValue=9.3308, parameterName=Glow_1}
   )
   "A == B"=@n7
   "A < B"=@n7
  )
  n9 = If(
   A=@n2
   B=Constant() {r=2}
   "A > B"=Multiply(
    A=VectorParameter() {group=Emissive_Winter, parameterName=ColorEmissive_Winter_2, defaultValue=<rgba 0.1936 0.6163 1 1>}.RGB
    B=ScalarParameter() {defaultValue=1.4859, parameterName=Glow_2}
   )
   "A == B"=@n8
   "A < B"=@n8
  )
  n10 = OneMinus(None=MaterialFunctionCall(
    Fade Length=ScalarParameter() {defaultValue=5000, parameterName=EmissiveDistanceFade}
    Fade Offset=ScalarParameter() {defaultValue=2000, parameterName=EmissiveDistanceFade_Offset}
   ) {materialFunction=&CameraDepthFade}.Result)
  n11 = Multiply(A=@n1.R, B=StaticSwitchParameter(True=Multiply(A=@n9, B=@n10), False=@n9) {parameterName=UseEmissiveDistancefade})
  FunctionOutput(None=@n11) {outputName=Emissive Only, sortPriority=3}
  n12 = FunctionInput(Preview=MakeMaterialAttributes!(
    BaseColor
    Metallic
    Specular
    Roughness
    Anisotropy
    EmissiveColor
    Opacity
    OpacityMask
    Normal
    Tangent
    WorldPositionOffset
    SubsurfaceColor
    ClearCoat
    ClearCoatRoughness
    AmbientOcclusion
    Refraction
    CustomizedUVs_0
    CustomizedUVs_1
    CustomizedUVs_2
    CustomizedUVs_3
    CustomizedUVs_4
    CustomizedUVs_5
    CustomizedUVs_6
    CustomizedUVs_7
    PixelDepthOffset
    ShadingModel
    Displacement
   )) {bUsePreviewValueAsDefault=true, inputType=FunctionInput_MaterialAttributes}
  FunctionOutput(None=StaticSwitchParameter(
    True=MaterialFunctionCall!(Material=@n12, Emissive=@n11) {materialFunction=/Engine/Functions/MaterialLayerFunctions/MatLayerBlend_Emissive}.Blended Material
    False=@n12
   ) {group=Emissive, parameterName=UseEmissiveTextureWinter}) {outputName=Material Output}
 ]}
